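# Physics asset for the Primal Falcon backpack that sets up bodies and constraints for its dynamic cape.
# Center, left and right chains of three capsules each hang from small MSTR bodies constrained to the pack, with narrow swing and twist limits along each chain.
# Angular-free cross constraints tie the outer chains to the center at the second and third segments.
# A two-capsule butt collider, also constrained to the pack, provides a body-shaped collision volume.
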PhysicsAsset /BRCosmetics/Accessories/FORT_Backpacks/Backpack_F_MED_Primal_Falcon/Meshes/F_MED_Primal_Falcon_Pack_Cloth_Physics
bodies {
 dyn_buttCollider [
  {shapeType=Capsule, length=16, radius=14.5, shapeName=dyn_buttCollider_capsule, center=<-0.75 0 0>, rotation=<rot 0 -90 0>}
  {shapeType=Capsule, length=15, radius=12, shapeName=dyn_buttCollider_capsule, rotation=<rot 90 0 0>}
 ]
 dyn_cape_01 [{shapeType=Capsule, length=5, radius=2, center=<4.0481 0 0>, rotation=<rot 90 0 0>}]
 dyn_cape_01_MSTR [{shapeType=Capsule, length=0.505, radius=0.505, rotation=<rot 0 -90 0>}]
 dyn_cape_01_l [{shapeType=Capsule, length=5, radius=2, center=<5.0013 0 0>, rotation=<rot 90 0 0>}]
 dyn_cape_01_l_MSTR [{shapeType=Capsule, length=0.505, radius=0.505, rotation=<rot 0 -90 0>}]
 dyn_cape_01_r [{shapeType=Capsule, length=5, radius=2, center=<4.544 0 0>, rotation=<rot 90 0 0>}]
 dyn_cape_01_r_MSTR [{shapeType=Capsule, length=0.505, radius=0.505, rotation=<rot 0 -90 0>}]
 dyn_cape_02 [{shapeType=Capsule, length=5, radius=2, center=<3.9676 0 0>, rotation=<rot 90 0 0>}]
 dyn_cape_02_l [{shapeType=Capsule, length=5, radius=2, center=<3.6134 0 0>, rotation=<rot 90 0 0>}]
 dyn_cape_02_r [{shapeType=Capsule, length=5, radius=2, center=<4.2747 0 0>, rotation=<rot 90 0 0>}]
 dyn_cape_03 [{shapeType=Capsule, length=6.25, radius=2, center=<2.9695 0 -0.0001>, rotation=<rot 90 0 0>}]
 dyn_cape_03_l [{shapeType=Capsule, length=6.25, radius=2, center=<2.9695 0 -0.0001>, rotation=<rot 90 0 0>}]
 dyn_cape_03_r [{shapeType=Capsule, length=6.25, radius=2, center=<2.9695 0 -0.0001>, rotation=<rot 90 0 0>}]
 pack [{shapeType=Capsule, length=4, radius=4, shapeName=pack_capsule}]
}
constraints [
 {bone1Name=dyn_cape_01_r, bone2Name=dyn_cape_01_r_MSTR, swing1LimitDegrees=5, swing2LimitDegrees=15, twistLimitDegrees=2}
 {bone1Name=dyn_cape_02_r, bone2Name=dyn_cape_01_r, swing1LimitDegrees=2, swing2LimitDegrees=30, twistLimitDegrees=2}
 {bone1Name=dyn_cape_03_r, bone2Name=dyn_cape_02_r, swing1LimitDegrees=2, swing2LimitDegrees=30, twistLimitDegrees=2}
 {bone1Name=dyn_cape_03_l, bone2Name=dyn_cape_02_l, swing1LimitDegrees=2, swing2LimitDegrees=30, twistLimitDegrees=2}
 {bone1Name=dyn_cape_02_l, bone2Name=dyn_cape_01_l, swing1LimitDegrees=2, swing2LimitDegrees=30, twistLimitDegrees=2}
 {bone1Name=dyn_cape_01_l, bone2Name=dyn_cape_01_l_MSTR, swing1LimitDegrees=5, swing2LimitDegrees=15, twistLimitDegrees=2}
 {bone1Name=dyn_cape_03, bone2Name=dyn_cape_02, swing1LimitDegrees=2, swing2LimitDegrees=30, twistLimitDegrees=2}
 {bone1Name=dyn_cape_02, bone2Name=dyn_cape_01, swing1LimitDegrees=2, swing2LimitDegrees=30, twistLimitDegrees=2}
 {bone1Name=dyn_cape_01, bone2Name=dyn_cape_01_MSTR, swing1LimitDegrees=5, swing2LimitDegrees=15, twistLimitDegrees=2}
 {
  bone1Name=dyn_cape_01_r_MSTR
  bone2Name=pack
  swing1LimitDegrees=45
  swing1Motion=Free
  swing2LimitDegrees=45
  swing2Motion=Free
  twistLimitDegrees=45
  twistMotion=Free
 }
 {
  bone1Name=dyn_cape_01_l_MSTR
  bone2Name=pack
  swing1LimitDegrees=45
  swing1Motion=Free
  swing2LimitDegrees=45
  swing2Motion=Free
  twistLimitDegrees=45
  twistMotion=Free
 }
 {
  bone1Name=dyn_cape_01_MSTR
  bone2Name=pack
  swing1LimitDegrees=45
  swing1Motion=Free
  swing2LimitDegrees=45
  swing2Motion=Free
  twistLimitDegrees=45
  twistMotion=Free
 }
 {
  bone1Name=dyn_buttCollider
  bone2Name=pack
  swing1LimitDegrees=45
  swing1Motion=Free
  swing2LimitDegrees=45
  swing2Motion=Free
  twistLimitDegrees=45
  twistMotion=Free
 }
 {
  bone1Name=dyn_cape_03_l
  bone2Name=dyn_cape_03
  swing1LimitDegrees=45
  swing1Motion=Free
  swing2LimitDegrees=45
  swing2Motion=Free
  twistLimitDegrees=45
  twistMotion=Free
 }
 {
  bone1Name=dyn_cape_03_r
  bone2Name=dyn_cape_03
  swing1LimitDegrees=45
  swing1Motion=Free
  swing2LimitDegrees=45
  swing2Motion=Free
  twistLimitDegrees=45
  twistMotion=Free
 }
 {
  bone1Name=dyn_cape_02_l
  bone2Name=dyn_cape_02
  swing1LimitDegrees=45
  swing1Motion=Free
  swing2LimitDegrees=45
  swing2Motion=Free
  twistLimitDegrees=45
  twistMotion=Free
 }
 {
  bone1Name=dyn_cape_02_r
  bone2Name=dyn_cape_02
  swing1LimitDegrees=45
  swing1Motion=Free
  swing2LimitDegrees=45
  swing2Motion=Free
  twistLimitDegrees=45
  twistMotion=Free
 }
]
