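# Physics asset for the wings of the Primal Falcon back bling, simulating sway in its mechanical arms and feathers.
# Each side's mech chain hangs from the pack via a free-rotation master joint, then bends on one narrow swing axis.
# Feather masters lock rotationally to the first mech segment, feeding two-segment capsule chains with looser tips.
# The left wing has three feather chains, the right two; a separate box body sways slightly on the pack.
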PhysicsAsset /BRCosmetics/Accessories/FORT_Backpacks/Backpack_F_MED_Primal_Falcon/Meshes/F_MED_Primal_Falcon_Pack_Wings_Physics
bodies {
 dyn_bone2_01 [{shapeType=Box, extentX=9.9682, extentY=2.9898, extentZ=11.2898, shapeName=dyn_bone2_01_box, center=<4 0 0>}]
 dyn_feather1_l_01 [{shapeType=Capsule, length=12, radius=3, shapeName=dyn_feather1_l_01_capsule, center=<2.5 0 0>, rotation=<rot 90 0 0>}]
 dyn_feather1_l_02 [{shapeType=Capsule, length=10.75, radius=4.25, shapeName=dyn_feather1_l_01_capsule, center=<3 0 0>, rotation=<rot 90 0 0>}]
 dyn_feather1_l_MSTR [{shapeType=Capsule, length=0.505, radius=0.505, shapeName=dyn_feather1_l_MSTR_capsule, rotation=<rot 0 -90 0>}]
 dyn_feather1_r_01 [{shapeType=Capsule, length=12, radius=3, shapeName=dyn_feather1_r_01_capsule, center=<3 0 0>, rotation=<rot 90 0 0>}]
 dyn_feather1_r_02 [{shapeType=Capsule, length=10.75, radius=4.25, shapeName=dyn_feather1_r_01_capsule, center=<2.5 0 0>, rotation=<rot 90 0 0>}]
 dyn_feather1_r_MSTR [{shapeType=Capsule, length=0.505, radius=0.505, shapeName=dyn_feather1_r_MSTR_capsule, rotation=<rot 0 -90 0>}]
 dyn_feather2_l_01 [{shapeType=Capsule, length=12, radius=3, shapeName=dyn_feather2_l_01_capsule, center=<1.5 0 0>, rotation=<rot 90 0 0>}]
 dyn_feather2_l_02 [{shapeType=Capsule, length=10.75, radius=4.25, shapeName=dyn_feather2_l_01_capsule, center=<1.5 0 0>, rotation=<rot 90 0 0>}]
 dyn_feather2_l_MSTR [{shapeType=Capsule, length=0.505, radius=0.505, shapeName=dyn_feather2_l_MSTR_capsule, rotation=<rot 0 -90 0>}]
 dyn_feather2_r_01 [{shapeType=Capsule, length=12, radius=3, shapeName=dyn_feather2_r_01_capsule, center=<2.75 0 0>, rotation=<rot 90 0 0>}]
 dyn_feather2_r_02 [{shapeType=Capsule, length=10.75, radius=4.25, shapeName=dyn_feather2_r_01_capsule, center=<3 0 0>, rotation=<rot 90 0 0>}]
 dyn_feather2_r_MSTR [{shapeType=Capsule, length=0.505, radius=0.505, shapeName=dyn_feather2_r_MSTR_capsule, rotation=<rot 0 -90 0>}]
 dyn_feather3_l_01 [{shapeType=Capsule, length=12, radius=3, shapeName=dyn_feather3_l_01_capsule, center=<3.5 0 0>, rotation=<rot 90 0 0>}]
 dyn_feather3_l_02 [{shapeType=Capsule, length=10.75, radius=4.25, shapeName=dyn_feather3_l_01_capsule, center=<3.5 0 0>, rotation=<rot 90 0 0>}]
 dyn_feather3_l_MSTR [{shapeType=Capsule, length=0.505, radius=0.505, shapeName=dyn_feather3_l_MSTR_capsule, rotation=<rot 0 -90 0>}]
 dyn_mech_l_01 [{
   shapeType=Capsule
   length=9.1751
   radius=7.9095
   shapeName=dyn_mech_l_01_capsule
   center=<3.7573 -0.3369 -6.2815>
   rotation=<rot -2.2942 -90 -27.8785>
  }]
 dyn_mech_l_02 [{shapeType=Capsule, length=9.0943, radius=1.9846, shapeName=dyn_mech_l_02_capsule, center=<2 0 0>, rotation=<rot 90 0 0>}]
 dyn_mech_l_MSTR [{shapeType=Capsule, length=0.505, radius=0.505, shapeName=dyn_mech_l_MSTR_capsule, rotation=<rot 0 -90 0>}]
 dyn_mech_r_01 [{
   shapeType=Capsule
   length=9.3645
   radius=7.909
   shapeName=dyn_mech_r_01_capsule
   center=<3.8341 -0.3285 6.4565>
   rotation=<rot 2.3083 -90 27.6754>
  }]
 dyn_mech_r_02 [{shapeType=Capsule, length=9.0943, radius=1.9846, shapeName=dyn_mech_r_02_capsule, center=<2 0 0>, rotation=<rot 90 0 0>}]
 dyn_mech_r_MSTR [{shapeType=Capsule, length=0.505, radius=0.505, shapeName=dyn_mech_r_MSTR_capsule, rotation=<rot 0 -90 0>}]
 pack [{shapeType=Capsule, length=4, radius=4}]
}
constraints [
 {bone1Name=dyn_bone2_01, bone2Name=pack, swing1LimitDegrees=3, swing2LimitDegrees=3, twistLimitDegrees=2}
 {
  bone1Name=dyn_mech_l_MSTR
  bone2Name=pack
  swing1LimitDegrees=45
  swing1Motion=Free
  swing2LimitDegrees=45
  swing2Motion=Free
  twistLimitDegrees=45
  twistMotion=Free
 }
 {
  bone1Name=dyn_mech_r_MSTR
  bone2Name=pack
  swing1LimitDegrees=45
  swing1Motion=Free
  swing2LimitDegrees=45
  swing2Motion=Free
  twistLimitDegrees=45
  twistMotion=Free
 }
 {bone1Name=dyn_mech_l_01, bone2Name=dyn_mech_l_MSTR, swing1LimitDegrees=0, swing2LimitDegrees=10, twistLimitDegrees=0}
 {bone1Name=dyn_mech_r_01, bone2Name=dyn_mech_r_MSTR, swing1LimitDegrees=0, swing2LimitDegrees=10, twistLimitDegrees=0}
 {bone1Name=dyn_mech_r_02, bone2Name=dyn_mech_r_01, swing1LimitDegrees=0, swing2LimitDegrees=10, twistLimitDegrees=0}
 {bone1Name=dyn_mech_l_02, bone2Name=dyn_mech_l_01, swing1LimitDegrees=0, swing2LimitDegrees=10, twistLimitDegrees=0}
 {
  bone1Name=dyn_feather1_l_MSTR
  bone2Name=dyn_mech_l_01
  swing1LimitDegrees=45
  swing1Motion=Locked
  swing2LimitDegrees=45
  swing2Motion=Locked
  twistLimitDegrees=45
  twistMotion=Locked
 }
 {
  bone1Name=dyn_feather1_r_MSTR
  bone2Name=dyn_mech_r_01
  swing1LimitDegrees=45
  swing1Motion=Locked
  swing2LimitDegrees=45
  swing2Motion=Locked
  twistLimitDegrees=45
  twistMotion=Locked
 }
 {
  bone1Name=dyn_feather2_r_MSTR
  bone2Name=dyn_mech_r_01
  swing1LimitDegrees=45
  swing1Motion=Locked
  swing2LimitDegrees=45
  swing2Motion=Locked
  twistLimitDegrees=45
  twistMotion=Locked
 }
 {
  bone1Name=dyn_feather2_l_MSTR
  bone2Name=dyn_mech_l_01
  swing1LimitDegrees=45
  swing1Motion=Locked
  swing2LimitDegrees=45
  swing2Motion=Locked
  twistLimitDegrees=45
  twistMotion=Locked
 }
 {
  bone1Name=dyn_feather3_l_MSTR
  bone2Name=dyn_mech_l_01
  swing1LimitDegrees=45
  swing1Motion=Locked
  swing2LimitDegrees=45
  swing2Motion=Locked
  twistLimitDegrees=45
  twistMotion=Locked
 }
 {bone1Name=dyn_feather1_l_01, bone2Name=dyn_feather1_l_MSTR, swing1LimitDegrees=3, swing2LimitDegrees=3, twistLimitDegrees=2}
 {bone1Name=dyn_feather2_l_01, bone2Name=dyn_feather2_l_MSTR, swing1LimitDegrees=3, swing2LimitDegrees=3, twistLimitDegrees=2}
 {bone1Name=dyn_feather3_l_01, bone2Name=dyn_feather3_l_MSTR, swing1LimitDegrees=3, swing2LimitDegrees=3, twistLimitDegrees=2}
 {bone1Name=dyn_feather1_r_01, bone2Name=dyn_feather1_r_MSTR, swing1LimitDegrees=3, swing2LimitDegrees=3, twistLimitDegrees=2}
 {bone1Name=dyn_feather2_r_01, bone2Name=dyn_feather2_r_MSTR, swing1LimitDegrees=3, swing2LimitDegrees=3, twistLimitDegrees=2}
 {bone1Name=dyn_feather1_l_02, bone2Name=dyn_feather1_l_01, swing1LimitDegrees=15, swing2LimitDegrees=8, twistLimitDegrees=2}
 {bone1Name=dyn_feather2_l_02, bone2Name=dyn_feather2_l_01, swing1LimitDegrees=20, swing2LimitDegrees=8, twistLimitDegrees=2}
 {bone1Name=dyn_feather3_l_02, bone2Name=dyn_feather3_l_01, swing1LimitDegrees=20, swing2LimitDegrees=8, twistLimitDegrees=2}
 {bone1Name=dyn_feather1_r_02, bone2Name=dyn_feather1_r_01, swing1LimitDegrees=15, swing2LimitDegrees=8, twistLimitDegrees=2}
 {bone1Name=dyn_feather2_r_02, bone2Name=dyn_feather2_r_01, swing1LimitDegrees=20, swing2LimitDegrees=8, twistLimitDegrees=2}
]
