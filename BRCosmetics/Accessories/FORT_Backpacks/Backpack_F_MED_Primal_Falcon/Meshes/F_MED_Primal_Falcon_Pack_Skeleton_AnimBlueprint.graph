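# Animation Blueprint for the Primal Falcon backpack, driving physics and skydive posing of its cape, mechanical wings and feathers.
# Each update copies the parent's pelvis transform to a collider bone, derives cape rotation from spine bend, and updates gravity through a deprecated helper.
# Separate wing and cloth RigidBody passes follow, with wing physics halved while skydiving.
# Constraints then tie feathers to mech bones, and additive offsets pose pack, mechs and feathers during dive or float.
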AnimBlueprint /BRCosmetics/Accessories/FORT_Backpacks/Backpack_F_MED_Primal_Falcon/Meshes/F_MED_Primal_Falcon_Pack_Skeleton_AnimBlueprint
target_skeleton=&/F_MED_Primal_Falcon_Pack_Skeleton
graphs [
 AnimGraph {nodes [
   n1 = "Transform (Modify) Bone - Bone: dyn_buttCollider"(
    ComponentPose=Local To Component(LocalPose=Local Space Ref Pose())
    Alpha=1
    Translation=Get pelvis_tran() -> (pelvis_tran:Vector)
    Rotation=Get pelvis_rot() -> (pelvis_rot:Rotator)
    Scale="1,1,1"
   ) {
    rotationMode=BMM_Replace
    rotationSpace=BCS_WorldSpace
    translationMode=BMM_Replace
    translationSpace=BCS_WorldSpace
    alphaBoolBlend {blendOption=Linear}
    boneToModify {boneName=dyn_buttCollider}
   }
   n2 = Get backpack_rotate()
   n3 = "Transform (Modify) Bone - Bone: dyn_cape_01"(ComponentPose=@n1, Alpha=1, Translation="0,0,0", Rotation=@n2, Scale="1,1,1") {rotationMode=BMM_Additive, alphaBoolBlend {blendOption=Linear}, boneToModify {boneName=dyn_cape_01}}
   n4 = RigidBody(
    ComponentPose="Transform (Modify) Bone - Bone: dyn_cape_01_l"(
     ComponentPose="Transform (Modify) Bone - Bone: dyn_cape_01_r"(ComponentPose=@n3, Alpha=1, Translation="0,0,0", Rotation=@n2, Scale="1,1,1") {rotationMode=BMM_Additive, alphaBoolBlend {blendOption=Linear}, boneToModify {boneName=dyn_cape_01_r}}
     Alpha=1
     Translation="0,0,0"
     Rotation=@n2
     Scale="1,1,1"
    ) {rotationMode=BMM_Additive, alphaBoolBlend {blendOption=Linear}, boneToModify {boneName=dyn_cape_01_l}}
    Alpha=Get Skydive Alpha() -> (Skydive Alpha:double)
   ) {
    bClampLinearTranslationLimitToRefPose=true
    bOverrideWorldGravity=true
    overridePhysicsAsset=&/F_MED_Primal_Falcon_Pack_Wings_Physics
    simulationSpace=BaseBoneSpace
    alphaBoolBlend {blendOption=Linear}
    baseBoneRef {boneName=pack}
    componentAppliedLinearAccClamp=<3000 3000 3000>
    componentLinearAccScale=<0.3 0.3 0.3>
    componentLinearVelScale=<0.1 0.1 0.1>
    overrideWorldGravity {z=-500}
   }
   n5 = RigidBody!(ComponentPose=@n4, Alpha, OverrideWorldGravity:Vector=Get Gravity Override(), ExternalForce) -> (Pose) {
    bClampLinearTranslationLimitToRefPose=true
    bOverrideWorldGravity=true
    overridePhysicsAsset=&/F_MED_Primal_Falcon_Pack_Cloth_Physics
    simulationSpace=BaseBoneSpace
    alphaBoolBlend {blendOption=Linear}
    baseBoneRef {boneName=pack}
    componentAppliedLinearAccClamp=<3000 3000 3000>
    componentLinearAccScale=<0.3 0.3 0.35>
    componentLinearVelScale=<0.3 0.3 0.3>
    overrideWorldGravity {z=-980}
   }
   n6 = "Constraint - dyn_feather1_l_01 (P)"(ComponentPose=@n5, Alpha=0.9, ConstraintWeights_0=1) {
    alphaBoolBlend {blendOption=Linear}
    boneToModify {boneName=dyn_feather1_l_01}
    constraintSetup [{targetBone {boneName=dyn_mech_l_01}}]
    constraintWeights [1]
   }
   n7 = "Constraint - dyn_feather2_l_01 (P)"(ComponentPose=@n6, Alpha=0.9, ConstraintWeights_0=1) {
    alphaBoolBlend {blendOption=Linear}
    boneToModify {boneName=dyn_feather2_l_01}
    constraintSetup [{targetBone {boneName=dyn_mech_l_01}}]
    constraintWeights [1]
   }
   n8 = "Constraint - dyn_feather3_l_01 (P)"(ComponentPose=@n7, Alpha=0.9, ConstraintWeights_0=1) {
    alphaBoolBlend {blendOption=Linear}
    boneToModify {boneName=dyn_feather3_l_01}
    constraintSetup [{targetBone {boneName=dyn_mech_l_01}}]
    constraintWeights [1]
   }
   n9 = "Constraint - dyn_feather1_r_01 (P)"(ComponentPose=@n8, Alpha=0.9, ConstraintWeights_0=1) {
    alphaBoolBlend {blendOption=Linear}
    boneToModify {boneName=dyn_feather1_r_01}
    constraintSetup [{targetBone {boneName=dyn_mech_r_01}}]
    constraintWeights [1]
   }
   n10 = "Constraint - dyn_feather2_r_01 (P)"(ComponentPose=@n9, Alpha=0.9, ConstraintWeights_0=1) {
    alphaBoolBlend {blendOption=Linear}
    boneToModify {boneName=dyn_feather2_r_01}
    constraintSetup [{targetBone {boneName=dyn_mech_r_01}}]
    constraintWeights [1]
   }
   n11 = Get Skydive Feathers Alpha() -> (Skydive Feathers Alpha:double)
   n12 = "Transform (Modify) Bone - Bone: pack"(ComponentPose=@n10, Alpha=@n11, Translation="0,-3,0", Rotation="0,0,0") {
    bAlphaBoolEnabled=false
    rotationMode=BMM_Additive
    rotationSpace=BCS_BoneSpace
    translationMode=BMM_Additive
    translationSpace=BCS_BoneSpace
    alphaBoolBlend {blendOption=Linear}
    boneToModify {boneName=pack}
    translation {y=-3}
   }
   n13 = "Transform (Modify) Bone - Bone: dyn_mech_l_MSTR"(ComponentPose=@n12, Alpha=@n11, Translation="0,0,0", Rotation="0,0,-5", Scale="1,1,1") {
    rotationMode=BMM_Additive
    rotationSpace=BCS_BoneSpace
    translationMode=BMM_Additive
    translationSpace=BCS_BoneSpace
    alphaBoolBlend {blendOption=Linear}
    boneToModify {boneName=dyn_mech_l_MSTR}
    rotation {roll=-5}
   }
   n14 = "Transform (Modify) Bone - Bone: dyn_mech_r_MSTR"(ComponentPose=@n13, Alpha=@n11, Translation="0,0,0", Rotation="0,0,5", Scale="1,1,1") {
    rotationMode=BMM_Additive
    rotationSpace=BCS_BoneSpace
    translationMode=BMM_Additive
    translationSpace=BCS_BoneSpace
    alphaBoolBlend {blendOption=Linear}
    boneToModify {boneName=dyn_mech_r_MSTR}
    rotation {roll=5}
   }
   n15 = "Transform (Modify) Bone - Bone: dyn_feather1_l_MSTR"(ComponentPose=@n14, Alpha=@n11, Translation="0,0,0", Rotation="20,0,-5", Scale="1,1,1") {
    bAlphaBoolEnabled=false
    rotationMode=BMM_Additive
    rotationSpace=BCS_BoneSpace
    translationMode=BMM_Additive
    translationSpace=BCS_BoneSpace
    alphaBoolBlend {blendOption=Linear}
    alphaScaleBiasClamp {bInterpResult=true}
    boneToModify {boneName=dyn_feather1_l_MSTR}
    rotation {pitch=20, roll=-5}
   }
   n16 = "Transform (Modify) Bone - Bone: dyn_feather2_l_MSTR"(ComponentPose=@n15, Alpha=@n11, Translation="0,0,0", Rotation="20,0,-5", Scale="1,1,1") {
    bAlphaBoolEnabled=false
    rotationMode=BMM_Additive
    rotationSpace=BCS_BoneSpace
    translationMode=BMM_Additive
    translationSpace=BCS_BoneSpace
    alphaBoolBlend {blendOption=Linear}
    alphaScaleBiasClamp {bInterpResult=true}
    boneToModify {boneName=dyn_feather2_l_MSTR}
    rotation {pitch=20, roll=-5}
   }
   n17 = "Transform (Modify) Bone - Bone: dyn_feather3_l_MSTR"(ComponentPose=@n16, Alpha=@n11, Translation="0,0,0", Rotation="20,0,-5", Scale="1,1,1") {
    bAlphaBoolEnabled=false
    rotationMode=BMM_Additive
    rotationSpace=BCS_BoneSpace
    translationMode=BMM_Additive
    translationSpace=BCS_BoneSpace
    alphaBoolBlend {blendOption=Linear}
    alphaScaleBiasClamp {bInterpResult=true}
    boneToModify {boneName=dyn_feather3_l_MSTR}
    rotation {pitch=20, roll=-5}
   }
   n18 = "Transform (Modify) Bone - Bone: dyn_feather1_r_MSTR"(ComponentPose=@n17, Alpha=@n11, Translation="0,0,0", Rotation="-20,0,5", Scale="1,1,1") {
    bAlphaBoolEnabled=false
    rotationMode=BMM_Additive
    rotationSpace=BCS_BoneSpace
    translationMode=BMM_Additive
    translationSpace=BCS_BoneSpace
    alphaBoolBlend {blendOption=Linear}
    alphaScaleBiasClamp {bInterpResult=true}
    boneToModify {boneName=dyn_feather1_r_MSTR}
    rotation {pitch=-20, roll=5}
   }
   n19 = "Transform (Modify) Bone - Bone: dyn_feather2_r_MSTR"(ComponentPose=@n18, Alpha=@n11, Translation="0,0,0", Rotation="-20,0,5", Scale="1,1,1") {
    bAlphaBoolEnabled=false
    rotationMode=BMM_Additive
    rotationSpace=BCS_BoneSpace
    translationMode=BMM_Additive
    translationSpace=BCS_BoneSpace
    alphaBoolBlend {blendOption=Linear}
    alphaScaleBiasClamp {bInterpResult=true}
    boneToModify {boneName=dyn_feather2_r_MSTR}
    rotation {pitch=-20, roll=5}
   }
   Output Pose(Result=Component To Local(ComponentPose=@n19)) {layerGroup=None}
  ]}
 EventGraph {nodes [
   Event BlueprintUpdateAnimation()
   > n1 = Sequence!(execute) -> (then_0, then_1, then_2:Exec, then_3:Exec)
   n2 = BreakTransform(InTransform=GetDeltaTransformFromRefPose(self=GetOwnerSkeletalMeshComponent(), BoneName=spine_03, BaseName=spine_01))
   n3 = MapRangeClamped(Value="Absolute (Float)"(A=BreakRotator(InRot=@n2.Rotation)), InRangeA=0, InRangeB=25, OutRangeB=30)
   Set backpack_rotate(execute=@n1.then_1, backpack_rotate=MakeRotator(Roll=float + float(A=0, B=@n3) {is=title K2Node_CommutativeAssociativeBinaryOperator}))
   n4 = Gravity Override State Function DEPRECATED(
    execute=@n1.then_0
    CharacterPartAnimInstance=Self-Reference()
    GameTimeInSeconds=GetGameTimeInSeconds()
    GravityOverrideParams=Get Gravity Override Params()
   )
   > Set Gravity Override(Gravity Override=@n4)
   n5 = GetPartSkeletalMeshComponent()
   Is Valid(exec=@n1.then_2, InputObject=@n5)
   n6 = GetAttachParent!(self:Scene Component Object Reference=@n5) -> (ReturnValue:Scene Component Object Reference)
   > Is Valid(InputObject=@n6)
   n7 = BreakTransform(InTransform=GetSocketTransform(self=@n6, InSocketName=pelvis, TransformSpace=RTS_World))
   > Set pelvis_tran(pelvis_tran:Vector=@n7) -> (Output_Get:Vector)
   > Set pelvis_rot(pelvis_rot:Rotator=@n7.Rotation) -> (Output_Get:Rotator)
   Set Skydive Alpha(execute=@n1.then_3, Skydive Alpha:double=SelectFloat(A=0.5, B=1, bPickA=Get bIsSkyDiving())) -> (Output_Get:double)
   n8 = Get bIsSkydiveDiveMode()
   n9 = Get bIsSkydiveFloating()
   > Set Skydive Feathers Alpha(Skydive Feathers Alpha:double=SelectFloat(A=1, B=0, bPickA=OR Boolean(A=@n8, B=@n9))) -> (Output_Get:double)
   TryGetPawnOwner()
  ]}
]
variables [
 backpack_rotate {type=Rotator}
 Gravity Override {type=Vector}
 Gravity Override Params {type=GravityOverrideParamsStruct}
 is female {type=boolean}
 pelvis_rot {type=Rotator}
 pelvis_tran {type=Vector}
 Skydive Alpha {type=number}
 Skydive Feathers Alpha {type=number}
]
